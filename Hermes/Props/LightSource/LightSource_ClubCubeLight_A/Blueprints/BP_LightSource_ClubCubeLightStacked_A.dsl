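; Stacked club cube light prop Blueprint that gives its two mesh components separate emissive tint colors.
; The construction script runs the parent script, then creates a dynamic material instance from slot 0 of each component's current material.
; It sets the SFX_Emissive_Tint(Expand) parameter to Primary Tint Color on StaticMeshComponent and Secondary Tint Color on Cubelight.
; BeginPlay, overlap and Tick are disabled stubs, so this graph adds no runtime logic.
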(event EventBeginPlay
  (|CallParentFunction) ; This node is disabled and will not be called.\nDrag off pins to build functionality.
)

(event Collision|EventActorBeginOverlap
  (params (OtherActor :type Actor%20Object%20Reference))
  (|CallParentFunction OtherActor) ; This node is disabled and will not be called.\nDrag off pins to build functionality.
)

(event EventTick
  (params (DeltaSeconds :type Float%20%28single-precision%29))
  (|CallParentFunction DeltaSeconds) ; This node is disabled and will not be called.\nDrag off pins to build functionality.
)

(fn ConstructionScript
  (|CallParentFunction)
  (bind _returnvalue (Rendering|Material|CreateDynamicMaterialInstance (Variables|Debug|GetStaticMeshComponent) 0 (Rendering|Material|GetMaterial (Variables|Debug|GetStaticMeshComponent) 0) "None" :!class /Script/Engine.PrimitiveComponent))
  (Rendering|Material|SetVectorParameterValue _returnvalue "SFX_Emissive_Tint(Expand)" (Variables|Default|GetPrimary%20Tint%20Color) :!class /Script/Engine.MaterialInstanceDynamic)
  (bind _returnvalue_2 (Rendering|Material|CreateDynamicMaterialInstance (Variables|Default|GetCubelight) 0 (Rendering|Material|GetMaterial (Variables|Default|GetCubelight) 0) "None" :!class /Script/Engine.PrimitiveComponent))
  (Rendering|Material|SetVectorParameterValue _returnvalue_2 "SFX_Emissive_Tint(Expand)" (Variables|Default|GetSecondary%20Tint%20Color) :!class /Script/Engine.MaterialInstanceDynamic))
